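# Test MetaSound one-shot mono source that switches MIDI files during playback into a Fusion Sampler using the Bloom SYN jam-loop patch.
# OnPlay starts a transport, driving a metronome clock and MIDI Player that feeds the sampler.
# A delayed trigger sets the array index to 1, swapping the player from the Bb file to the Bbm file.
# A clock timestamp trigger at bar 3 fires OnFinished.
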MetaSoundSource /Harmonix/Tests/HarmonixMetasound/FusionSamplerNode/MS_Test_FusionSamplerNode_JamLoops_ChangeMidi
inputs [
 MIDI File {type=MIDIAsset:Array}
 UE.Source.OnPlay {type=Trigger}
]
interfaces [UE.OutputFormat.Mono:1.0, UE.Source.OneShot:1.0, UE.Source:1.1]
outputs [
 AudioOut {type=Audio}
 "UE.OutputFormat.Mono.Audio:0" {type=Audio}
 UE.Source.OneShot.OnFinished {type=Trigger}
]
pages [Default {
  input_defaults {
   UE.Source.OnPlay=null
   MIDI File [
    /Harmonix/Tests/HarmonixMetasound/FusionSamplerNode/Midi/Bloom_SYN_Bb_92
    /Harmonix/Tests/HarmonixMetasound/FusionSamplerNode/Midi/Bloom_SYN_Bbm_92
   ]
  }
  nodes [
   TriggerDelay_0 = UE.Trigger Delay(In=^UE.Source.OnPlay)
   n1 = Value.Value:Int32(Set=@TriggerDelay_0.Out) {also {defaults {Target Value=1}}}
   n2 = HarmonixNodes.TriggerToTransport(Play=^UE.Source.OnPlay)
   n3 = HarmonixNodes.Metronome(Transport=@n2.Transport)
   HarmonixNodesFusionSampler_0 = HarmonixNodes.FusionSampler(MIDI Stream=HarmonixNodes.MIDIPlayer(
     MIDI File=Array.Get:MIDIAsset:Array(Array=^MIDI File, Trigger=@n1.On Set, Index=@n1.Output Value).Element
     MIDI Clock=@n3.MIDI Clock
     Transport=@n2.Transport
    ).MIDI Stream) {also {defaults {Patch=/Harmonix/Tests/HarmonixMetasound/FusionSamplerNode/Patches/Bloom_SYN_Bb_92_precat}, version [0,1]}}
   HarmonixNodesMidiClockTimestampTrigger_0 = HarmonixNodes.MidiClockTimestampTrigger(MIDI Clock=@n3.MIDI Clock, Music Timestamp=HarmonixNodes.MusicTimestampBuilder() {also {defaults {Bar=3}, version [0,1]}}.Music Timestamp)
  ]
  links [
   {dst="@AudioOut", src=HarmonixNodesFusionSampler_0.Audio Mono}
   {dst="@UE.OutputFormat.Mono.Audio:0", src=HarmonixNodesFusionSampler_0.Audio Mono}
   {dst="@UE.Source.OneShot.OnFinished", src=HarmonixNodesMidiClockTimestampTrigger_0.Trigger Out}
  ]
 }]
properties {
 assetImportData=&self:AssetImportData
 sampleRate=48000
}
