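# Test MetaSound source for the Harmonix Fusion Sampler node, named for jam-loop transposition changes, though it exposes no transposition input.
# On Play drives a transport shared by a Metronome, clocked from the Tempo input, and a MIDI Player playing the Bloom_SYN_Bb_92 file.
# The player feeds a stereo Fusion Sampler with the matching precat patch.
# Left goes to the mono source output, right to a separate AudioOut output.
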MetaSoundSource /Harmonix/Tests/HarmonixMetasound/FusionSamplerNode/MS_Test_FusionSamplerNode_JamLoops_ChangeTransposition
inputs [
 Tempo {type=Float}
 UE.Source.OnPlay {type=Trigger}
]
interfaces [UE.OutputFormat.Mono:1.0, UE.Source:1.1]
outputs [
 AudioOut {type=Audio}
 "UE.OutputFormat.Mono.Audio:0" {type=Audio}
]
pages [Default {
  input_defaults {Tempo=84.48, UE.Source.OnPlay=null}
  nodes [
   n1 = HarmonixNodes.TriggerToTransport(Play=^UE.Source.OnPlay)
   HarmonixNodesFusionSamplerStereo_0 = HarmonixNodes.FusionSamplerStereo(MIDI Stream=HarmonixNodes.MIDIPlayer(MIDI Clock=HarmonixNodes.Metronome(Tempo=^Tempo, Transport=@n1.Transport).MIDI Clock, Transport=@n1.Transport) {also {defaults {MIDI File=/Harmonix/Tests/HarmonixMetasound/FusionSamplerNode/Midi/Bloom_SYN_Bb_92}}}.MIDI Stream) {also {
     defaults {
      Patch=/Harmonix/Tests/HarmonixMetasound/FusionSamplerNode/Patches/Bloom_SYN_Bb_92_precat
      Use Pitch Shifters="<MetaSoundBuilderResult.SUCCEEDED: 0>"
     }
     version [0,1]
    }}
  ]
  links [
   {dst="@UE.OutputFormat.Mono.Audio:0", src=HarmonixNodesFusionSamplerStereo_0.Audio Left}
   {dst="@AudioOut", src=HarmonixNodesFusionSamplerStereo_0.Audio Right}
  ]
 }]
properties {
 assetImportData=&self:AssetImportData
 bLooping=true
 duration=10000
 sampleRate=48000
}
