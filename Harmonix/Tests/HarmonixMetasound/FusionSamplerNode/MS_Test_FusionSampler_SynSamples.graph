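# Mono MetaSound source test asset for the Harmonix Fusion Sampler, rendering a synth MIDI clip through a sample patch.
# On Play both starts the transport and triggers a seek to a fixed bar-beat target at beat 3.
# The transport drives a MIDI Player, whose stream feeds the Fusion Sampler and its mono output.
# The one-shot OnFinished output is left unconnected.
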MetaSoundSource /Harmonix/Tests/HarmonixMetasound/FusionSamplerNode/MS_Test_FusionSampler_SynSamples
inputs [UE.Source.OnPlay {type=Trigger}]
interfaces [UE.OutputFormat.Mono:1.0, UE.Source.OneShot:1.0, UE.Source:1.1]
outputs [
 "UE.OutputFormat.Mono.Audio:0" {type=Audio}
 UE.Source.OneShot.OnFinished {type=Trigger}
]
pages [Default {
  input_defaults {UE.Source.OnPlay=null}
  nodes [
   n1 = HarmonixNodes.BarBeatToSeekTarget() {also {defaults {Beat=3}, version [0,1]}}
   HarmonixNodesFusionSampler_0 = HarmonixNodes.FusionSampler(MIDI Stream=HarmonixNodes.MIDIPlayer(Transport=HarmonixNodes.TriggerToTransport(Play=^UE.Source.OnPlay, Trigger Seek=^UE.Source.OnPlay, Seek Target=@n1.Seek Target).Transport) {also {defaults {MIDI File=/Harmonix/Tests/HarmonixMetasound/FusionSamplerNode/Midi/Bloom_SYN_Bb_92}}}.MIDI Stream) {also {defaults {Patch=/Harmonix/Tests/HarmonixMetasound/FusionSamplerNode/Patches/Bloom_SYN_Bb_92_precat}, version [0,1]}}
  ]
  links [{dst="@UE.OutputFormat.Mono.Audio:0", src=HarmonixNodesFusionSampler_0.Audio Mono}]
 }]
properties {assetImportData=&self:AssetImportData, sampleRate=48000}
